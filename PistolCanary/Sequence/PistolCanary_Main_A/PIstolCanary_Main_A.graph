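# Master Level Sequence for the PistolCanary "Main A" segment, staging a tentacle boss encounter.
# Subsequence tracks run the main narrative moment and soundtrack across the whole segment, with hero attacks A, B and C as three short windows within it.
# Many short, partly overlapping tentacle attack-cycle subsequences are chained along the timeline.
# The sequence itself keys only the boss's yaw swings and holds constant transforms for the boss and a GrillMoon POI audio volume.
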LevelSequence /PistolCanary/Sequence/PistolCanary_Main_A/PIstolCanary_Main_A
bindings [
 BP_GrillMoon_POI_Volume {
  id=#1
  possessed_class=BP_PistolCanary_AudioVolume_C
  tracks [{
    type=MovieScene3DTransformTrack
    sections [{
      type=MovieScene3DTransformSection
      channels [
       Location.X {type=MovieSceneScriptingDoubleChannel, default=18536.626, const [14384.627,1]}
       Location.Y {type=MovieSceneScriptingDoubleChannel, default=63350.062, const [69690.0625,1]}
       Location.Z {type=MovieSceneScriptingDoubleChannel, default=10972, const [7976,1]}
       Rotation.X {type=MovieSceneScriptingDoubleChannel, default=0, const [0,1]}
       Rotation.Y {type=MovieSceneScriptingDoubleChannel, default=0, const [0,1]}
       Rotation.Z {type=MovieSceneScriptingDoubleChannel, default=0, const [0,1]}
       Scale.X {type=MovieSceneScriptingDoubleChannel, default=1, const [1,1]}
       Scale.Y {type=MovieSceneScriptingDoubleChannel, default=1, const [1,1]}
       Scale.Z {type=MovieSceneScriptingDoubleChannel, default=1, const [1,1]}
      ]
     }]
   }]
 }
 TentacleBoss {
  id=#2
  possessed_class=BGA_Tentacle_C
  tracks [{
    type=MovieScene3DTransformTrack
    sections [{
      type=MovieScene3DTransformSection
      channels [
       Location.X {type=MovieSceneScriptingDoubleChannel, default=123.5116, const [123.5116,1]}
       Location.Y {type=MovieSceneScriptingDoubleChannel, default=-6719.1338, const [-6719.1338,1]}
       Location.Z {type=MovieSceneScriptingDoubleChannel, default=900, const [900,1]}
       Rotation.X {type=MovieSceneScriptingDoubleChannel, default=0, const [0,1]}
       Rotation.Y {type=MovieSceneScriptingDoubleChannel, default=0, const [0,1]}
       Rotation.Z {
        type=MovieSceneScriptingDoubleChannel
        default=30
        thinned_from=25
        keys [[3600,0],[4719,0],[5176,-30],[5325,15],[5534,-15],[5571,-30],[5811,30],[6129,30],[6309,0],[8430,0],[8610,-30],[8934,-30],[9174,15],[9269,15],[9302,0],[9405,15],[9480,30],[11614,30],[11832,0],[12640,0],[13497,-30],[13969,30],[14562,-30],[14668,0]]
       }
       Scale.X {type=MovieSceneScriptingDoubleChannel, default=0.75, const [0.75,1]}
       Scale.Y {type=MovieSceneScriptingDoubleChannel, default=0.75, const [0.75,1]}
       Scale.Z {type=MovieSceneScriptingDoubleChannel, default=0.75, const [0.75,1]}
      ]
     }]
   }]
 }
]
display_rate {denominator=1, numerator=60}
tick_resolution {denominator=1, numerator=24000}
tracks [
 {type=MovieSceneSubTrack, display_name=mnm - main narrative moment, sections [{type=MovieSceneSubSection, end_frame=15518, start_frame=3600}]}
 {type=MovieSceneSubTrack, display_name=haa - hero attack A, sections [{type=MovieSceneSubSection, end_frame=7181, start_frame=6723}]}
 {type=MovieSceneSubTrack, display_name=hab - hero attack B, sections [{type=MovieSceneSubSection, end_frame=10417, start_frame=9939}]}
 {type=MovieSceneSubTrack, display_name=hac - hero attack C, sections [{type=MovieSceneSubSection, end_frame=15235, start_frame=14668}]}
 {type=MovieSceneSubTrack, display_name=mas - main A soundtrack, sections [{type=MovieSceneSubSection, end_frame=15518, start_frame=3600}]}
 {
  type=MovieSceneSubTrack
  display_name=tac - tentacle attack cycle - Animation
  sections [
   {type=MovieSceneSubSection, end_frame=11832, start_frame=11614}
   {type=MovieSceneSubSection, end_frame=15518, start_frame=15130}
   {type=MovieSceneSubSection, end_frame=6748, start_frame=6230}
   {type=MovieSceneSubSection, end_frame=4267, start_frame=3749}
   {type=MovieSceneSubSection, end_frame=5694, start_frame=5176}
   {type=MovieSceneSubSection, end_frame=9567, start_frame=9049}
   {type=MovieSceneSubSection, end_frame=8513, start_frame=7604}
   {type=MovieSceneSubSection, end_frame=11300, start_frame=10398}
   {type=MovieSceneSubSection, end_frame=7604, start_frame=7184}
   {type=MovieSceneSubSection, end_frame=6230, start_frame=5694}
   {type=MovieSceneSubSection, end_frame=8610, start_frame=8430}
   {type=MovieSceneSubSection, end_frame=9049, start_frame=8513}
   {type=MovieSceneSubSection, end_frame=12256, start_frame=11720}
   {type=MovieSceneSubSection, end_frame=9174, start_frame=8934}
   {type=MovieSceneSubSection, end_frame=9962, start_frame=9542}
   {type=MovieSceneSubSection, end_frame=13060, start_frame=12640}
   {type=MovieSceneSubSection, end_frame=13969, start_frame=13060}
   {type=MovieSceneSubSection, end_frame=11720, start_frame=11300}
   {type=MovieSceneSubSection, end_frame=5176, start_frame=4267}
   {type=MovieSceneSubSection, end_frame=12640, start_frame=12256}
   {type=MovieSceneSubSection, end_frame=4010, start_frame=3670}
   {type=MovieSceneSubSection, end_frame=14756, start_frame=14562}
   {type=MovieSceneSubSection, end_frame=14562, start_frame=13969}
   {type=MovieSceneSubSection, end_frame=6309, start_frame=6129}
   {type=MovieSceneSubSection, end_frame=5811, start_frame=5571}
   {type=MovieSceneSubSection, end_frame=7184, start_frame=6748}
   {type=MovieSceneSubSection, end_frame=15130, start_frame=14694}
   {type=MovieSceneSubSection, end_frame=10398, start_frame=9962}
  ]
 }
 {type=MovieSceneSubTrack, display_name=tac - tentacle attack cycle - Animation, sections []}
]
